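# Niagara ribbon trail for the Lexa (Female) pickaxe, evidently shown during its swing animations.
# A single attached-ribbon emitter spawns short-lived particles positioned from the User.SkeletalMesh data interface at both spawn and update.
# User.RibbonColor, red-orange by default, tints particles and ribbon material, with RGB boosted for brightness.
# Curves keyed to emitter loop age, not particle age, widen the ribbon over the loop's first half and fade alpha over the second.
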NiagaraSystem /BRCosmetics/Weapons/FORT_Melee/Pickaxe_Lexa_Female/FX/NS_Pickaxe_Lexa_Animtrail
dependencies {usedDataInterfaces [&NiagaraDataInterfaceCurve], usedDynamicInputs [], usedModules [&SystemState]}
emitters {FN_AttachedRibbon_v3 {
  emitterUpdateScript [
   &EmitterState()
   &FN_SpawnRibbon_v2(SpawnRate=60)
  ]
  particleSpawnScript [/Game/Effects/Niagara/Modules/Ribbons/FN_InitializeRibbon_v4(
    Color:LinearColor=$User.RibbonColor
    Enable Skeletal Position=true
    Facing Axis:ENiagaraOrientationAxis=::ZAxis
    Lifetime=0.25
    Ribbon Width=9
    SkeletalMesh:NiagaraDataInterfaceSkeletalMesh=$User.SkeletalMesh
   )]
  particleUpdateScript [
   /Game/Effects/Niagara/Modules/Ribbons/FN_UpdateRibbon_v4(Enable Skeletal Position=true, SkeletalMesh:NiagaraDataInterfaceSkeletalMesh=$User.SkeletalMesh)
   /Game/Effects/Niagara/Modules/Materials/FN_RibbonMaterial/FN_RibbonMaterial_v2(Color:LinearColor=$User.RibbonColor)
   &ScaleColor(
    Scale Alpha=&FloatFromCurve(FloatCurve=DataInterface() {propertyValues {Curve {keys [<at 0.5 1>,<at 1 0>]}}}, CurveIndex=$Emitter.NormalizedLoopAge)
    Scale RGB=&MakeVectorFromLinearColorRGB(Color=<rgba 4 4 4 1>)
    ScaleA=true
    ScaleRGB=true
   )
   &ScaleRibbonWidth(Ribbon Width Scale=&FloatFromCurve(
     FloatCurve=DataInterface() {propertyValues {Curve {keys [{arriveTangent=4.9397, interpMode=RCIM_Cubic, leaveTangent=6.9794, tangentMode=RCTM_User, time=0, value=0}, <at 0.5 1 RCIM_Cubic>]}}}
     CurveIndex=$Emitter.NormalizedLoopAge
    ))
  ]
  rendererClasses [&NiagaraRibbonRendererProperties]
  renderers [{rendererClass=&NiagaraRibbonRendererProperties}]
 }}
summary {
 systemName=NS_Pickaxe_Lexa_Animtrail
 emitters [{emitterName=FN_AttachedRibbon_v3, rendererClasses [&NiagaraRibbonRendererProperties]}]
 userVariables [
  User.RibbonColor {type=LinearColor, description="", defaultValue {struct=LinearColor, value=<rgba 0.7031 0.0417 0.0162 1>}}
  User.RibbonIDOffset {type=NiagaraInt32}
  User.SkeletalMesh {
   type=NiagaraDataInterfaceSkeletalMesh
   defaultValue {
    struct=NiagaraExt_VariableValue_DataInterface
    value {dataInterface=&self:NiagaraDataInterfaceSkeletalMesh_0}
   }
  }
 ]
}
