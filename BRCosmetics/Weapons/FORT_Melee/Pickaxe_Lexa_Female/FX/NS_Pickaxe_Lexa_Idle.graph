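# Idle Niagara effect for the female Lexa pickaxe, providing a colored light and glow on the weapon.
# Its single emitter spawns one particle on the first loop only, which persists past its lifetime and tracks an offset sample on User.SkeletalMesh every update.
# User.LightColor tints that particle, which feeds both a light renderer and a component renderer.
# User.PlayerSpeed and User.bIsFrontEndPreview are exposed, but no shown module reads them.
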NiagaraSystem /BRCosmetics/Weapons/FORT_Melee/Pickaxe_Lexa_Female/FX/NS_Pickaxe_Lexa_Idle
dependencies {usedDataInterfaces [], usedDynamicInputs [], usedModules [&SystemState]}
emitters {Light_And_Glow {
  emitterUpdateScript [
   &EmitterState()
   &SpawnBurst_Instantaneous(Spawn Count=1, Spawn Probability=HlslExpression() {hlslExpression="Emitter.LoopCount == 0"}, Use Spawn Probability=true)
  ]
  particleSpawnScript [/Niagara/Modules/Spawn/Initialization/InitializeParticle(
    Color=$User.LightColor
    Module.Write Color=true
    Module.Write Lifetime=true
    Module.Write Mass=false
    Module.Write Material Random=true
    Module.Write Position=false
    Module.Write Scale=false
    Module.Write SpriteRotation=false
    Module.Write SpriteSize=true
    Sprite Size=<75 75>
   )]
  particleUpdateScript [
   &ParticleState(Kill Particles When Lifetime Has Elapsed=false, Let Infinitely Lived Particles Die When Emitter Deactivates=true)
   &SkeletalMeshLocation(
    Apply Sampled Orientation to Mesh Particles=true
    Bone Sampling Mode=::NewEnumerator4
    Inherit Velocity=false
    MaskSpawnGroup=false
    Sampled Orientation Coordinate Space=::World
    Sampled Position Coordinate Space=::World
    Sampled Position Offset=<16 0 0>
    Sampled Position Offset Coordinate Space:ENiagaraMeshTransforms=::NewEnumerator3
    Sampled Velocity Coordinate Space:ENiagaraCoordinateSpace=::World
    Skeletal Mesh=$User.SkeletalMesh
   )
   &Color(Color=$User.LightColor)
   &FN_Light_v3(
    Color=$User.LightColor
    Create Color Attribute=true
    Create Exponent Attribute=true
    Create Radius Attribute=true
    Exponent=2
    Radius=30
    Scale Color=6
   )
  ]
  rendererClasses [&NiagaraLightRendererProperties, &NiagaraComponentRendererProperties]
  renderers [
   {rendererClass=&NiagaraLightRendererProperties, rendererIndex=0}
   {rendererClass=&NiagaraComponentRendererProperties, rendererIndex=1}
  ]
 }}
summary {
 systemName=NS_Pickaxe_Lexa_Idle
 emitters [{
   emitterName=Light_And_Glow
   rendererClasses [&NiagaraLightRendererProperties, &NiagaraComponentRendererProperties]
  }]
 userVariables [
  User.bIsFrontEndPreview {type=NiagaraBool}
  User.LightColor {type=LinearColor, description="", defaultValue {struct=LinearColor, value=<rgba 1 1 1 1>}}
  User.PlayerSpeed {type=NiagaraFloat, defaultValue=0, description=""}
  User.SkeletalMesh {
   type=NiagaraDataInterfaceSkeletalMesh
   defaultValue {
    struct=NiagaraExt_VariableValue_DataInterface
    value {dataInterface=&self:NiagaraDataInterfaceSkeletalMesh_0}
   }
  }
 ]
}
